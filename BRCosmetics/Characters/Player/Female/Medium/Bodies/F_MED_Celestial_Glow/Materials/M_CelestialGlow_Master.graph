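# Master skeletal-mesh material for the Celestial Glow character body, adding an animated black hole eclipse to emissive within the SkinFX mask.
# It projects the effect onto a camera-facing virtual plane at BlackholeLocation, skewing UVs by view angle against BlackholeForward.
# Sphere masks, wobbling polar-coordinate noise and curve-colored flowmap turbulence build the shape, rim and glow.
# These layer over CharacterSkinFX output, with reduced specular, before the character PostFX functions.
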Material /BRCosmetics/Characters/Player/Female/Medium/Bodies/F_MED_Celestial_Glow/Materials/M_CelestialGlow_Master
graph {
 properties {
  bUseMaterialAttributes=true
  bUsedWithMeshDeformer=true
  bUsedWithMorphTargets=true
  bUsedWithSkeletalMesh=true
  pixelDepthOffsetMode=PDOM_Legacy
  refractionMethod=RM_IndexOfRefraction
  usedShadingModels=Default Lit
 }
 nodes [
  n1 = MaterialFunctionCall() {materialFunction=&CameraDirectionVector}
  n2 = Constant()
  n3 = Constant3Vector() {constant=<rgba 0 1 0 0>}
  n4 = If(
   A=ComponentMask(None=@n1.Result) {r=true}
   B=@n2
   "A > B"=@n2
   "A == B"=Multiply(A=Transform(None=@n3) {desc=Up camera direction, transformSourceType=TRANSFORMSOURCE_View}, B=Constant() {r=0.01})
   "A < B"=@n2
  ) {equalsThreshold=0.005}
  n5 = MaterialFunctionCall(
   Plane Normal Axis=Normalize(VectorInput=Add(A=@n1.Result, B=@n4))
   Plane Center=VectorParameter() {parameterName=BlackholeLocation, defaultValue=<rgba 0 0 120 1>}.RGB
   UV Scale=ScalarParameter() {defaultValue=12, parameterName=BlackholeSize}
  ) {materialFunction=&VirtualPlaneCoordinates}
  n6 = Normalize(VectorInput=Add(B=@n5.UVs) {constA=-0.5})
  n7 = Multiply(A=ComponentMask(None=@n6) {g=true}, B=ComponentMask(None=@n6) {r=true})
  Divide(A=@n7) {constB=-0.5}
  n8 = MaterialFunctionCall() {materialFunction=&BaseFN_CharacterFunction_PreFX}
  n9 = GetMaterialAttributes(None=MaterialFunctionCall(MaterialAttributes=@n8.Result) {materialFunction=&CharacterSkinFX}.Result) {attributeGetTypes [#1, #2]}
  n10 = Divide(A=@n5.UVs) {constB=3}
  n11 = Constant() {r=20}
  n12 = Time()
  n13 = ComponentMask(None=MaterialFunctionCall!(
    Volume Texture=TextureObject() {texture=/FortWater/Textures/T_Turbulence_Seamless}
    UV=Rotator(Coordinate=@n10, Time=@n11) {centerX=0.25, centerY=0.25}
    Velocity=Multiply(A=@n6) {constB=-1}
    Override Time=Divide(A=@n12)
    Mip Level
   ) {materialFunction=/Engine/Functions/Engine_MaterialFunctions02/Texturing/Flowmaps_2D}.Result RGB) {r=true}
  n14 = CurveAtlasRowParameter(CurveTime=@n13) {atlas=/Game/Athena/LinearColorCurves/VFX_CurveAtlas, curve=/Game/Athena/LinearColorCurves/Curve_GhostChain_FireTrail, defaultValue=19}
  n15 = Multiply(A=Multiply(A=@n14, B=@n13), B=NamedRerouteUsage())
  n16 = OneMinus(None=NamedRerouteUsage())
  n17 = Power(Base=OneMinus(None=@n7)) {constExponent=10}
  n18 = Power(Base=TextureSample(Tex=TextureObjectParameter() {group=Skin FX, parameterName=SkinFX_Mask}).G)
  n19 = Multiply(A=Multiply(A=Multiply(A=@n15, B=@n16), B=Saturate(None=@n17)), B=@n18)
  n20 = Multiply(A=NamedRerouteUsage(), B=@n18)
  n21 = ScalarParameter() {defaultValue=0.9, parameterName=Blackhole Visibility}
  n22 = Multiply(A=ScalarParameter() {defaultValue=4, parameterName=Blackhole Back Glow Intensity}, B=NamedRerouteUsage())
  n23 = LinearInterpolate(
   A=LinearInterpolate(A=@n9.Emissive Color, B=NamedRerouteUsage(), Alpha=Multiply(A=@n20, B=@n21)) {constB=0}
   B=Multiply(A=@n22, B=VectorParameter() {parameterName="Eclipse Color && Hue Shift", defaultValue=<rgba 0 1.35 3 -0.1063>}.RGB) {constB=0.1}
   Alpha=Multiply(A=@n18, B=@n22)
  )
  n24 = MaterialFunctionCall(Material=SetMaterialAttributes!(MaterialAttributes=@n9.MaterialAttributes, Emissive Color=Add(A=@n19, B=@n23), Specular=Multiply(A=@n9.Specular) {constB=0.1}) {attributeSetTypes [#1, #2]}) {materialFunction=&BaseFN_CharacterFunction_PostFX}
  MaterialFunctionCall(In=@n24.Result) {materialFunction=&CINE_MF_CharPostFX}
  n25 = Multiply(A=DotProduct(
    A=MaterialFunctionCall() {materialFunction=&CameraDirectionVector}.Result
    B=VectorParameter() {parameterName=BlackholeForward, defaultValue=<rgba 0 1 0 1>}.RGB
   )) {constB=-1}
  NamedRerouteDeclaration(None=Abs(None=@n25)) {name=Blackhole Camera Vector, nodeColor=<rgba 0 1 0.3203 1>}
  NamedRerouteDeclaration(None=Sign(None=@n25)) {name=Blackhole Camera Sign, nodeColor=<rgba 1 0.4688 0 1>}
  n26 = Constant2Vector() {g=0.5, r=0.5}
  n27 = Subtract(A=@n5.UVs) {constB=0.5}
  n28 = Divide(A=ComponentMask(None=@n27) {g=true}) {constB=-10}
  n29 = Multiply(A=Multiply(A=ScalarParameter() {defaultValue=0.5, parameterName=ScewY}, B=NamedRerouteUsage()), B=ComponentMask(None=@n27) {r=true})
  n30 = NamedRerouteUsage()
  n31 = Add(
   A=@n5.UVs
   B=MaterialFunctionCall(Y=LinearInterpolate(A=@n28, B=@n29, Alpha=@n30)) {materialFunction=&MakeFloat2}.Result
  )
  NamedRerouteDeclaration(None=SphereMask(A=@n31, B=@n26, Radius=Constant() {r=1.05}, Hardness=Constant() {r=0.5})) {name=Blackhole Shape Soft, nodeColor=<rgba 0.4531 0 1 1>}
  n32 = VectorParameter() {parameterName="Eclipse Color && Hue Shift", defaultValue=<rgba 0 1.35 3 -0.1063>}
  n33 = NamedRerouteUsage()
  n34 = Constant() {r=0.5}
  n35 = NamedRerouteUsage()
  n36 = Distance(A=@n35, B=@n34)
  n37 = Add(
   A=Multiply(A=Time(), B=ScalarParameter() {defaultValue=0.2, parameterName=Eclipse Wiggle Speed, sliderMax=1})
   B=Saturate(None=Divide(A=@n36, B=@n34))
  )
  n38 = Rotator(
   Coordinate=@n35
   Time=Multiply(A=Sine(None=@n37), B=ScalarParameter() {defaultValue=0.1, parameterName=Eclipse Wiggle Range, sliderMax=1}) {constB=0.3}
  ) {speed=1}
  n39 = MaterialFunctionCall!(XY=Subtract(A=@n38) {constB=0.5}) {materialFunction=/BRCosmetics/Athena/Effects/Particles/Skydiving/NS_Contrail_DarkAndLight/Functions/MF_CartesianToPolar}
  n40 = Multiply(A=Sine(None=Subtract(A=@n39.Angle) {constB=0.25}), B=ScalarParameter() {defaultValue=1, parameterName=Bend, sliderMax=1})
  n41 = Multiply(A=Multiply(A=@n39.Radius, B=@n39.Radius), B=@n39.Radius)
  n42 = MaterialFunctionCall() {materialFunction=/BRCosmetics/Athena/Effects/Particles/Skydiving/NS_Contrail_DarkAndLight/MF_DarkAndLight_TimeOffset}
  n43 = ScalarParameter() {defaultValue=0.5, parameterName=Noise Pan Speed}
  n44 = AppendVector(
   A=Subtract(A=@n39.Angle, B=Multiply(A=@n40, B=@n41))
   B=Subtract(A=Add(A=@n39.Radius, B=@n39.Radius), B=Multiply(A=@n42.Time with Offset, B=@n43))
  )
  n45 = AppendVector(
   A=ScalarParameter() {defaultValue=2, parameterName=UV Scale U, sliderMax=2}
   B=ScalarParameter() {defaultValue=2, parameterName=UV Scale V, sliderMax=2}
  )
  n46 = NamedRerouteUsage()
  n47 = TextureSample!(UVs=Frac(None=Multiply(A=@n44, B=@n45)), Tex, "DDX(UVs)"=DDX(Value=@n46), "DDY(UVs)"=DDY(Value=@n46), Apply View MipBias) {
   mipValueMode=TMVM_Derivative
   samplerType=SAMPLERTYPE_LinearGrayscale
   texture=/BRCosmetics/Athena/Effects/Particles/Skydiving/NS_Contrail_DarkAndLight/T_DarkAndLight_Noise
  }
  n48 = ScalarParameter() {defaultValue=33, parameterName=Eclipse Power A, sliderMax=66}
  n49 = ScalarParameter() {defaultValue=2, parameterName=Eclipse Power B, sliderMax=66}
  n50 = MaterialFunctionCall!(
   Gradient=MaterialFunctionCall!(UV=@n38) {materialFunction=/BRCosmetics/Athena/Effects/Particles/Skydiving/NS_Contrail_DarkAndLight/Functions/MF_GradientDiagonal}."0, 0, 1, 1"
   Center Point=Constant() {r=0.5}
  ) {materialFunction=/Game/Effects/Fort_Effects/MaterialFunctions/MF_GradientReflected}
  n51 = Saturate(None=Power(
    Base=MaterialFunctionCall!(Base=@n33, Blend=@n47.R) {materialFunction=/BRCosmetics/Athena/Effects/Particles/Skydiving/NS_Contrail_DarkAndLight/Functions/MF_BlendOverlay_Float}.Result
    Exp=LinearInterpolate(A=@n48, B=@n49, Alpha=@n50.Result)
   ))
  n52 = Saturate(None=Subtract(A=@n51, B=NamedRerouteUsage()))
  NamedRerouteDeclaration(None=MaterialFunctionCall(Hue Shift Percentage=Multiply(A=@n32.A, B=@n52), Texture=Multiply(A=@n52, B=@n32.RGB)) {materialFunction=&HueShift}.Result) {name=Blackhole Glow, nodeColor=<rgba 0.7109 1 0 1>}
  NamedRerouteDeclaration(None=@n31) {name=Blackhole UVs, nodeColor=<rgba 0 0.5781 1 1>}
  NamedRerouteDeclaration(None=@n51) {name=Blackhole Mask, nodeColor=<rgba 0.6875 0 1 1>}
  n53 = Power(Base=SphereMask(A=@n31, B=@n26, Radius=Constant() {r=1.25}, Hardness=Constant())) {constExponent=3}
  n54 = NamedRerouteDeclaration(None=SphereMask(A=@n31, B=@n26, Radius=Constant() {r=0.5})) {name=Blackhole Shape, nodeColor=<rgba 0.0897 0 0.1979 1>}
  NamedRerouteDeclaration(None=Saturate(None=Subtract(A=@n53, B=@n54))) {name=Blackhole Shape Back Glow, nodeColor=<rgba 0.4531 0 1 1>}
  TextureCoordinate()
  TextureSample!(UVs, Tex, Apply View MipBias) {texture=&T_Inky_Smoke_Tile}
 ]
}
