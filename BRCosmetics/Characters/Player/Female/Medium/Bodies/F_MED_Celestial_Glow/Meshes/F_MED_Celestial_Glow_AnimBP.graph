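# Animation Blueprint for the Celestial Glow female medium cosmetic body, copying the attached parent's pose through a Control Rig.
# Additive bone offsets then move arm affectors by alphas mapped from clavicle yaw, and the right leg affector at constant strength.
# Each update computes those alphas, plus an unused right-knee value.
# It also sends spine_04's location and right vector to materials as BlackholeLocation and BlackholeForward, after a zero-length delay.
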AnimBlueprint /BRCosmetics/Characters/Player/Female/Medium/Bodies/F_MED_Celestial_Glow/Meshes/F_MED_Celestial_Glow_AnimBP
target_skeleton=&/F_MED_Celestial_Glow_Skeleton
graphs [
 AnimGraph {nodes [
   n1 = Control Rig(Source=Copy Pose From Mesh() {bUseAttachedParent=true}) {
    alphaBoolBlend {blendOption=Linear}
    controlRigAssetReference {blueprintRigClass=&/F_MED_Celestial_Glow_CtrlRig.F_MED_Celestial_Glow_CtrlRig_C}
   }
   n2 = "Transform (Modify) Bone - Bone: dyn_arm_affector_l"(
    ComponentPose=Local To Component(LocalPose=@n1)
    Alpha=Get L_Arm_Rotation() -> (L_Arm_Rotation:double)
    Translation="-8,2,0"
    Rotation="0,0,0"
    Scale="1,1,1"
   ) {
    translationMode=BMM_Additive
    translationSpace=BCS_BoneSpace
    alphaBoolBlend {blendOption=Linear}
    boneToModify {boneName=dyn_arm_affector_l}
    translation=<-8 2>
   }
   n3 = "Transform (Modify) Bone - Bone: dyn_arm_affector_r"(ComponentPose=@n2, Alpha=Get R_Arm_Rotation() -> (R_Arm_Rotation:double), Translation="12,10,0", Rotation="0,0,0", Scale="1,1,1") {
    translationMode=BMM_Additive
    translationSpace=BCS_BoneSpace
    alphaBoolBlend {blendOption=Linear}
    boneToModify {boneName=dyn_arm_affector_r}
    translation=<12 10>
   }
   n4 = "Transform (Modify) Bone - Bone: dyn_leg_affector_r"!(ComponentPose=@n3, bAlphaBoolEnabled:bool=True, Translation="0,-10,-15", Rotation="0,0,0", Scale="1,1,1") -> (Pose) {
    alphaInputType=Bool
    translationMode=BMM_Additive
    translationSpace=BCS_BoneSpace
    alphaBoolBlend {blendOption=Linear}
    boneToModify {boneName=dyn_leg_affector_r}
    translation {y=-10, z=-15}
   }
   Output Pose(Result=Component To Local(ComponentPose=@n4)) {layerGroup=None}
   Get R_Leg_Rotation() -> (R_Leg_Rotation:double)
  ]}
 EventGraph {nodes [
   Event BlueprintUpdateAnimation()
   > n1 = Sequence()
   n2 = GetPartSkeletalMeshComponent()
   > Is Valid(InputObject=@n2)
   > Delay(Duration=0)
   n3 = GetSocketTransform(self=@n2, InSocketName=spine_04, TransformSpace=RTS_World)
   > SetVectorParameterValueOnMaterials!(
    execute
    self:Mesh Component Object Reference=@n2
    ParameterName:Name=BlackholeLocation
    ParameterValue:Vector=TransformLocation!(T:"Transform (by ref)"=@n3, Location:Vector="0,4.7324, 0") -> (ReturnValue:Vector)
   ) -> (then)
   > SetVectorParameterValueOnMaterials(
    self=@n2
    ParameterName=BlackholeForward
    ParameterValue=GetRightVector!(InRot:Rotator=TransformRotation!(T:"Transform (by ref)"=@n3, Rotation:Rotator="0, 0, 0") -> (ReturnValue:Rotator)) -> (ReturnValue:Vector)
   )
   n4 = BreakTransform(InTransform=GetDeltaTransformFromRefPose(self=GetOwnerSkeletalMeshComponent(), BoneName=clavicle_l, BaseName=spine_05))
   Set L_Arm_Rotation(execute=@n1.then_1, L_Arm_Rotation:double=MapRangeClamped(Value=BreakRotator(InRot=@n4.Rotation).Yaw, InRangeA=8, InRangeB=16)) -> (Output_Get:double)
   n5 = BreakTransform(InTransform=GetDeltaTransformFromRefPose(self=GetOwnerSkeletalMeshComponent(), BoneName=clavicle_r, BaseName=spine_05))
   > Set R_Arm_Rotation(R_Arm_Rotation:double=MapRangeClamped(Value=BreakRotator(InRot=@n5.Rotation).Yaw, InRangeA=-2, InRangeB=15)) -> (Output_Get:double)
   n6 = BreakTransform(InTransform=GetDeltaTransformFromRefPose(self=GetOwnerSkeletalMeshComponent(), BoneName=calf_r, BaseName=thigh_r))
   > Set R_Leg_Rotation(R_Leg_Rotation:double=MapRangeClamped(Value=BreakRotator(InRot=@n6.Rotation).Yaw, InRangeA=50, InRangeB=110)) -> (Output_Get:double)
  ]}
]
variables [
 Crouch Vector {type=Vector}
 L_Arm_Rotation {type=number}
 R_Arm_Rotation {type=number}
 R_Leg_Rotation {type=number}
]
