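# Head Animation Blueprint for the Celestial Glow female body that drives a blackhole effect in the character part's materials.
# On each animation update it calls the parent, validates the part's skeletal mesh component and waits through a zero-length Delay.
# It then reads the spine_04 socket's world transform, writing an offset point as BlackholeLocation and the socket's right vector as BlackholeForward to the mesh materials.
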AnimBlueprint /BRCosmetics/Characters/Player/Female/Medium/Bodies/F_MED_Celestial_Glow/Meshes/F_MED_Celestial_Glow_Head_AnimBP
target_skeleton=/Game/Characters/Player/Common/Fortnite_Base_Head/Fortnite_Base_Head_Export_Skeleton
graphs [EventGraph {nodes [
   n1 = Event BlueprintUpdateAnimation()
   > Parent: BlueprintUpdateAnimation!(execute, DeltaTimeX:float=@n1.DeltaTimeX) -> (then) {is=title K2Node_CallParentFunction}
   n2 = GetPartSkeletalMeshComponent()
   > Is Valid(InputObject=@n2)
   > Delay(Duration=0)
   n3 = GetSocketTransform(self=@n2, InSocketName=spine_04, TransformSpace=RTS_World)
   > SetVectorParameterValueOnMaterials!(
    execute
    self:Mesh Component Object Reference=@n2
    ParameterName:Name=BlackholeLocation
    ParameterValue:Vector=TransformLocation!(T:"Transform (by ref)"=@n3, Location:Vector="0,4.7324, 0") -> (ReturnValue:Vector)
   ) -> (then)
   > SetVectorParameterValueOnMaterials(
    self=@n2
    ParameterName=BlackholeForward
    ParameterValue=GetRightVector!(InRot:Rotator=TransformRotation!(T:"Transform (by ref)"=@n3, Rotation:Rotator="0, 0, 0") -> (ReturnValue:Rotator)) -> (ReturnValue:Vector)
   )
  ]}]
variables []
