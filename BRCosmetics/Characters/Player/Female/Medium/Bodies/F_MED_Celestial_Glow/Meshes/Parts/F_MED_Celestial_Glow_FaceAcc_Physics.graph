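# Physics asset for the face accessory part of the Celestial Glow medium female player cosmetic, defining bodies and joint limits for its dynamic bones.
# One large capsule covers the head, and small capsules cover dyn bones grouped front, back, left, right and top.
# Each dyn bone is constrained directly to the head rather than chained, with per-bone swing limits and uniformly tight twist limits.
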PhysicsAsset /BRCosmetics/Characters/Player/Female/Medium/Bodies/F_MED_Celestial_Glow/Meshes/Parts/F_MED_Celestial_Glow_FaceAcc_Physics
bodies {
 dyn_back_01 [{shapeType=Capsule, length=0.51, radius=0.875, center=<0.8086 0.0001 0>, rotation=<rot 90 0 0>}]
 dyn_front_01 [{shapeType=Capsule, length=1.375, radius=1.875, center=<3.3481 0.0001 0>, rotation=<rot 90 0 0>}]
 dyn_front_02 [{shapeType=Capsule, length=2.4656, radius=1.4156, center=<2.6986 0.0001 0>, rotation=<rot 90 0 0>}]
 dyn_left_01 [{shapeType=Capsule, length=1.13, radius=0.88, center=<1.4944 0.0001 0>, rotation=<rot 90 0 0>}]
 dyn_left_02 [{shapeType=Capsule, length=0.755, radius=0.755, center=<0.911 0.0001 0>, rotation=<rot 90 0 0>}]
 dyn_left_03 [{shapeType=Capsule, length=0.51, radius=0.875, center=<0.8919 0.0001 0>, rotation=<rot 90 0 0>}]
 dyn_left_04 [{shapeType=Capsule, length=0.51, radius=0.875, center=<0.8098 0.0001 0.0001>, rotation=<rot 90 0 0>}]
 dyn_right_01 [{shapeType=Capsule, length=1.88, radius=1.255, center=<2.8799 0 0>, rotation=<rot 90 0 0>}]
 dyn_right_02 [{shapeType=Capsule, length=2.25, radius=1.375, center=<2.8104 0.0002 0>, rotation=<rot 90 0 0>}]
 dyn_right_03 [{shapeType=Capsule, length=1.26, radius=0.75, center=<1.7833 0.0001 0>, rotation=<rot 90 0 0>}]
 dyn_right_04 [{shapeType=Capsule, length=0.505, radius=0.505, center=<0.9991 0.0001 0>, rotation=<rot 90 0 0>}]
 dyn_top_01 [{shapeType=Capsule, length=0.38, radius=1.255, center=<1.3747 0 0>, rotation=<rot 90 0 0>}]
 dyn_top_02 [{shapeType=Capsule, length=0.75, radius=1.25, center=<1.3206 0.0001 0>, rotation=<rot 90 0 0>}]
 head [{shapeType=Capsule, length=6.0015, radius=6.9378, center=<3.8013 1.7345 0>, rotation=<rot 90 0 0>}]
}
constraints [
 {bone1Name=dyn_front_01, bone2Name=head, swing1LimitDegrees=15, swing2LimitDegrees=45, twistLimitDegrees=5}
 {bone1Name=dyn_front_02, bone2Name=head, swing1LimitDegrees=15, swing2LimitDegrees=35, twistLimitDegrees=5}
 {bone1Name=dyn_top_01, bone2Name=head, swing1LimitDegrees=18, swing2LimitDegrees=30, twistLimitDegrees=5}
 {bone1Name=dyn_right_01, bone2Name=head, swing1LimitDegrees=15, swing2LimitDegrees=30, twistLimitDegrees=5}
 {bone1Name=dyn_right_02, bone2Name=head, swing1LimitDegrees=15, swing2LimitDegrees=20, twistLimitDegrees=5}
 {bone1Name=dyn_left_01, bone2Name=head, swing1LimitDegrees=5, swing2LimitDegrees=25, twistLimitDegrees=5}
 {bone1Name=dyn_left_02, bone2Name=head, swing1LimitDegrees=10, swing2LimitDegrees=20, twistLimitDegrees=5}
 {bone1Name=dyn_left_03, bone2Name=head, swing1LimitDegrees=5, swing2LimitDegrees=15, twistLimitDegrees=5}
 {bone1Name=dyn_left_04, bone2Name=head, swing1LimitDegrees=5, swing2LimitDegrees=15, twistLimitDegrees=5}
 {bone1Name=dyn_back_01, bone2Name=head, swing1LimitDegrees=10, swing2LimitDegrees=25, twistLimitDegrees=5}
 {bone1Name=dyn_right_03, bone2Name=head, swing1LimitDegrees=10, swing2LimitDegrees=35, twistLimitDegrees=5}
 {bone1Name=dyn_top_02, bone2Name=head, swing1LimitDegrees=15, swing2LimitDegrees=35, twistLimitDegrees=5}
 {bone1Name=dyn_right_04, bone2Name=head, swing1LimitDegrees=10, swing2LimitDegrees=35, twistLimitDegrees=5}
]
